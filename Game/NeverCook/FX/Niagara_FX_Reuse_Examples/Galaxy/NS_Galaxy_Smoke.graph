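# Niagara system for a galaxy-themed smoke effect, built from one sprite emitter, Galaxy_World, in the FX reuse examples folder.
# It spawns a slow, steady stream of pale, translucent sprites around a small sphere, with randomized size and fixed lifetime.
# A curve grows each sprite quickly then shrinks it, alongside color and alpha scaling.
# User parameters for vehicle, skydiving and skeletal mesh are declared, but no shown module reads them.
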NiagaraSystem /Game/NeverCook/FX/Niagara_FX_Reuse_Examples/Galaxy/NS_Galaxy_Smoke
dependencies {
 usedDataInterfaces [&NiagaraDataInterfaceColorCurve, &NiagaraDataInterfaceCurve]
 usedDynamicInputs []
 usedModules [&SystemState]
}
emitters {Galaxy_World {
  emitterUpdateScript [
   &EmitterState(Life Cycle Mode=::NewEnumerator1, Loop Duration=1, UseLoopDelay=false)
   &SpawnRate(SpawnRate=5, Use Spawn Probability=false)
  ]
  particleSpawnScript [
   &FN_SetSpawnGroup(Number of Spawn Groups=5)
   &InitializeParticle(
    Color=<rgba 0.72 0.72 0.8 0.15>
    Lifetime Max=3
    Lifetime Min=3
    Lifetime Mode=::NewEnumerator1
    Sprite Rotation Mode=::NewEnumerator2
    Sprite Size=&Multiply_Vector2D_ByFloat(Float=&RandomRangeFloat(Minimum=7, Maximum=8), Vector2D=<2.5 2.5>)
    Sprite Size Mode=::NewEnumerator4
    UsePositionOffset=false
    Write Material Random=true
   )
   &ShapeLocation(MaskSpawnGroup=false, Override Seed=false, Sphere Radius=10, UseOwnerScale=true)
  ]
  particleUpdateScript [
   &ParticleState()
   &SolveForcesAndVelocity(Clamp Velocity=false, Limit Acceleration=false)
   &ScaleColor(ScaleA=true, ScaleRGB=true)
   &ScaleSpriteSize(Scale Factor=&Vector2DFromFloat(Value=&FloatFromCurve(FloatCurve=DataInterface() {propertyValues {Curve {keys [
          {arriveTangent=2.3875, interpMode=RCIM_Cubic, leaveTangent=2.3875, tangentMode=RCTM_User, time=0, value=0}
          {arriveTangent=0.5, interpMode=RCIM_Cubic, leaveTangent=0.5, time=0.3, value=1}
          {arriveTangent=-2.3244, interpMode=RCIM_Cubic, leaveTangent=-2.3244, tangentMode=RCTM_User, time=1, value=0.5}
         ]}}})))
  ]
  rendererClasses [&NiagaraSpriteRendererProperties]
  renderers [{rendererClass=&NiagaraSpriteRendererProperties}]
 }}
summary {
 systemName=NS_Galaxy_Smoke
 emitters [{emitterName=Galaxy_World, rendererClasses [&NiagaraSpriteRendererProperties]}]
 userVariables [
  User.bisInVehicle {type=NiagaraBool}
  User.IsSkydiving {type=NiagaraBool}
  User.Skeletal Mesh {
   type=NiagaraDataInterfaceSkeletalMesh
   defaultValue {
    struct=NiagaraExt_VariableValue_DataInterface
    value {dataInterface=&self:NiagaraDataInterfaceSkeletalMesh_0}
   }
  }
 ]
}
